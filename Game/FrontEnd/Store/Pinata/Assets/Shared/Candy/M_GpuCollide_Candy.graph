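# Translucent, lit particle material for candy sprites in the front-end store pinata, usable with Niagara and Cascade sprites.
# A SubUV candy texture is slightly desaturated, tinted by particle color and passed through an emissive adjustment function, alongside a brightened color branch and a normal map.
# Alpha follows particle alpha and, off mobile, blends toward a slightly reduced value as particle speed rises; mobile uses plain texture alpha.
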Material /Game/FrontEnd/Store/Pinata/Assets/Shared/Candy/M_GpuCollide_Candy
graph {
 properties {
  bUsedWithNiagaraSprites=true
  bUsedWithParticleSprites=true
  blendMode=BLEND_Translucent
  pixelDepthOffsetMode=PDOM_Legacy
  translucencyLightingMode=TLM_VolumetricPerVertexDirectional
  usedShadingModels=Default Lit
 }
 nodes [
  n1 = ParticleColor()
  n2 = TextureSample() {texture=&/T_Candy_SubUV}
  n3 = ScalarParameter() {defaultValue=0.05, parameterName=Fraction}
  MaterialFunctionCall(Emissive=Multiply(A=@n1.RGB, B=Desaturation(None=@n2.RGB, Fraction=@n3))) {materialFunction=&MF_FX_EmissiveAdjustments}
  Multiply(A=@n2.RGB) {constB=1.5}
  n4 = Multiply(A=@n2.A) {constB=0.85}
  n5 = Divide(A=ParticleSpeed(), B=ScalarParameter() {defaultValue=300, parameterName=Speed})
  Multiply(A=@n1.A, B=ShadingPathSwitch(Default=LinearInterpolate(A=@n2.A, B=@n4, Alpha=@n5), Mobile=@n2.A) {desc=Migrated from FeatureLevelSwitch})
  TextureSample() {samplerType=SAMPLERTYPE_Normal, texture=&/T_Candy_SubUV_N}
 ]
}
