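# Subsurface material for sparkly gumdrop candy in the front-end store piñata, flagged for Niagara mesh particles.
# Gumdrop textures drive tintable color, normal and surface response, with Fresnel-weighted sugar grain, subsurface color and optional particle-color tint.
# Switchable Fresnel glow and a camera-oriented panning sweep feed emissive.
# Sparkles multiply two thresholded, world-aligned noise samples on pre-skinned coordinates, shifted oppositely by camera direction, then add to base color.
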Material /Game/FrontEnd/Store/Pinata/Assets/Shared/Candy/M_Gumdrop_candy_Sparkles
graph {
 properties {
  bScreenSpaceReflections=true
  bUseMaterialAttributes=true
  bUseTranslucencyVertexFog=false
  bUsedWithNiagaraMeshParticles=true
  pixelDepthOffsetMode=PDOM_Legacy
  shadingModel=MSM_Subsurface
  translucencyLightingMode=TLM_VolumetricDirectional
  translucencyPass=MTP_BeforeDOF
  translucentSelfShadowDensityScale=1
  translucentSelfShadowSecondDensityScale=1
  translucentSelfShadowSecondOpacity=5
  translucentShadowDensityScale=1
  usedShadingModels=Subsurface
 }
 nodes [
  Clamp()
  n1 = Multiply(
   A=MaterialFunctionCall() {materialFunction=&CameraDirectionVector}.Result
   B=ScalarParameter() {defaultValue=0.1, group=Sparkles, parameterName=Sparkles_Camera_Movement}
  )
  ComponentMask(None=@n1) {g=true, r=true}
  Constant3Vector() {constant=<rgba 0.8333 0.004 0.004 1>}
  Constant3Vector() {constant=<rgba 0.4167 0.0038 0.0038 1>}
  Constant3Vector() {constant=<rgba 0.775 0.3774 0.327 1>}
  Constant() {r=3.5}
  Constant() {r=5}
  Constant() {r=1.25}
  n2 = ScalarParameter() {defaultValue=0.75, parameterName=NormalsTileU}
  n3 = TextureCoordinate()
  n4 = ComponentMask(None=@n3) {r=true}
  n5 = ComponentMask(None=@n3) {g=true}
  n6 = ScalarParameter() {defaultValue=1.5, parameterName=NormalsTileV}
  n7 = Add(A=AppendVector(A=Multiply(A=@n2, B=@n4), B=Multiply(A=@n5, B=@n6)), B=ScalarParameter() {defaultValue=0.2, parameterName=NormalsOffset})
  n8 = TextureCoordinate() {coordinateIndex=1}
  LinearInterpolate(
   A=Constant3Vector() {constant=<rgba 0 0 1 0>}
   B=Multiply(
    A=TextureSample(UVs=@n7) {samplerType=SAMPLERTYPE_Normal, texture=&/T_Crinkle_N}.RGB
    B=VectorParameter() {parameterName=NormalsPower, defaultValue=<rgba -2 -2 1 1>}.RGB
   )
   Alpha=Power(Base=ComponentMask(None=@n8) {g=true}, Exp=ScalarParameter() {defaultValue=3, parameterName=NormalsMaskPower}) {constExponent=5}
  )
  n9 = Constant() {r=1}
  n10 = PreSkinnedPosition()
  n11 = PreSkinnedLocalBounds()
  n12 = Divide(A=Subtract(A=VertexInterpolator(VS=@n10).PS, B=@n11.Min), B=Subtract(A=@n11.Max, B=@n11.Max))
  n13 = Multiply(A=ScalarParameter() {defaultValue=0.75, group=Sparkles, parameterName=Sparkles_Tile_Rate}, B=@n12)
  n14 = Constant() {r=1}
  n15 = VertexInterpolator(VS=PreSkinnedNormal())
  n16 = TextureObjectParameter() {
   group=Sparkles
   parameterName=Sparkles_textures
   samplerType=SAMPLERTYPE_LinearColor
   texture=/Game/Global/Textures/Default/Noise/T_Wrap_SnowStreak_Noise
  }
  n17 = MaterialFunctionCall(Color=MaterialFunctionCall(TextureObject=@n16, TextureSize=@n9, WorldPosition=Add(A=@n13, B=@n1), World Space Normal=@n15.PS, ProjectionTransitionContrast=@n14) {materialFunction=&WorldAlignedTexture}.XYZ Texture) {materialFunction=&SplitComponents}
  n18 = Constant() {r=-0.95}
  n19 = Add(A=Saturate(None=Add(A=@n17.B, B=@n18)), B=Saturate(None=Add(A=@n17.G, B=@n18)))
  n20 = Saturate(None=Add(A=@n17.R, B=@n18))
  n21 = OneMinus(None=@n1)
  n22 = MaterialFunctionCall(Color=MaterialFunctionCall(TextureObject=@n16, TextureSize=@n9, WorldPosition=Add(A=@n21, B=@n13), World Space Normal=@n15.PS, ProjectionTransitionContrast=@n14) {materialFunction=&WorldAlignedTexture}.XYZ Texture) {materialFunction=&SplitComponents}
  n23 = Saturate(None=Add(A=@n22.R, B=@n18))
  n24 = Add(A=Saturate(None=Add(A=@n22.G, B=@n18)), B=Saturate(None=Add(A=@n22.B, B=@n18)))
  n25 = Multiply(A=Multiply(A=Add(A=@n19, B=@n20), B=Add(A=@n23, B=@n24)), B=Constant() {r=500}) {constB=3}
  n26 = Subtract(A=Constant() {r=1}, B=Saturate(None=@n25))
  n27 = LinearInterpolate(Alpha=@n26) {constA=1, constB=0.5}
  n28 = TextureSample() {texture=&/T_Gumdrop_D}
  n29 = ComponentMask(None=TextureCoordinate()) {g=true}
  n30 = LinearInterpolate(A=Constant() {r=-3}, B=Constant() {r=3.5}, Alpha=Add(A=@n29) {constB=-0.15})
  n31 = Constant() {r=-3}
  n32 = Constant() {r=2}
  n33 = Add(A=@n29) {constB=0.5}
  n34 = LinearInterpolate(
   A=@n28.RGB
   B=LinearInterpolate(A=@n28.RGB, B=@n28.RGB, Alpha=Clamp(None=@n30))
   Alpha=Clamp(None=LinearInterpolate(A=@n31, B=@n32, Alpha=@n33))
  )
  n35 = StaticSwitchParameter(
   True=Multiply(A=VectorParameter() {parameterName=BaseColor, defaultValue=<rgba 0.6514 0.0168 0 1>}.RGB, B=Desaturation(None=@n34))
   False=@n34
  ) {defaultValue=true, parameterName=UseDiffuseColorParam}
  n36 = TextureCoordinate()
  n37 = TextureObject() {samplerType=SAMPLERTYPE_LinearGrayscale, texture=&/T_GumDrop_M}
  n38 = TextureSample(UVs=Multiply(A=@n36, B=ScalarParameter() {defaultValue=2, parameterName=Sparkles_Tile_Rate A}), Tex=@n37) {texture=&T_Gumdrop_D}
  n39 = Multiply(A=Multiply(A=OneMinus(None=@n38.R), B=TextureSample(UVs=@n36, Tex=@n37) {texture=&T_Gumdrop_D}.R)) {constB=0.5}
  n40 = Multiply(A=Add(A=@n28.B) {constB=-0.02}) {constB=3}
  n41 = Add(
   A=Multiply(A=Add(A=@n38.R, B=@n39), B=ScalarParameter() {defaultValue=0.8, parameterName=Sugar_opacity}) {constB=0.25}
   B=Multiply(A=Saturate(None=@n40), B=ScalarParameter() {defaultValue=0.5, parameterName=Sugar_opacity B}) {constB=0.1}
  )
  n42 = Multiply(
   A=Fresnel(
    ExponentIn=ScalarParameter() {defaultValue=3, group=Fresnel, parameterName=Sugar_ExponentIn}
    BaseReflectFractionIn=ScalarParameter() {defaultValue=0.02, group=Fresnel, parameterName=Sugar_BaseReflectFractionIn}
   )
   B=ScalarParameter() {defaultValue=3, parameterName=sugar_mulplyer}
  ) {constB=10}
  n43 = Add(A=@n35, B=Multiply(A=Saturate(None=@n41), B=Saturate(None=@n42))) {constB=0}
  n44 = Multiply(A=@n43, B=ParticleColor().RGB)
  n45 = ScalarParameter() {parameterName=UseParticleColor}
  n46 = TextureSample() {samplerType=SAMPLERTYPE_Masks, texture=&/T_Gumdrop_S}
  n47 = ScalarParameter() {parameterName=Metal}
  n48 = ScalarParameter() {defaultValue=0.5, parameterName=Spec}
  n49 = ScalarParameter() {defaultValue=0.87, parameterName=Rough}
  n50 = Fresnel(
   ExponentIn=ScalarParameter() {defaultValue=5, group=Fresnel, parameterName=ExponentIn}
   BaseReflectFractionIn=ScalarParameter() {defaultValue=0.04, group=Fresnel, parameterName=BaseReflectFractionIn}
  )
  n51 = VectorParameter() {group=Fresnel, parameterName=EmissiveColor, defaultValue=<rgba 1 1 1 0>}
  n52 = StaticSwitchParameter(True=Multiply(A=@n35, B=Multiply(A=@n50, B=@n51.RGB)), False=Constant()) {group=Fresnel, parameterName=UseFresnelEmissive}
  n53 = Multiply(
   A=Subtract(
    A=WorldPosition().XYZ
    B=MaterialFunctionCall() {materialFunction=&ComponentPivotLocation}.Result
   )
   B=ScalarParameter() {defaultValue=0.004, group=Sweep, parameterName=WorldTileScale}
  )
  n54 = Normalize(VectorInput=Transform(None=VectorParameter() {group=Sweep, parameterName=SweepDirectionCamSpace, defaultValue=<rgba -0.5 0.45 -1 0>}.RGB) {transformSourceType=TRANSFORMSOURCE_Camera})
  n55 = ComponentMask(None=Multiply(A=DotProduct(A=@n53, B=@n54), B=@n54)) {b=true, r=true}
  n56 = ScalarParameter() {defaultValue=0.18, group=Sweep, parameterName=PanSpeed}
  n57 = Frac(None=ComponentMask(None=Panner(Coordinate=@n55, Speed=@n56) {speedX=0.2, speedY=0.2}) {g=true})
  n58 = Floor(None=Add(A=@n57) {constB=0.03})
  n59 = Subtract(A=@n57) {constB=0.97}
  n60 = Power(Base=OneMinus(None=Multiply(A=@n59) {constB=33.3333})) {constExponent=3}
  n61 = StaticSwitchParameter(
   True=Multiply(A=Multiply(A=@n58, B=@n60), B=VectorParameter() {group=Sweep, parameterName=SweepColour, defaultValue=<rgba 2 1.6277 0.6 0>}.RGB)
   False=Constant()
  ) {group=Sweep, parameterName=UseSweep}
  n62 = ScalarParameter() {defaultValue=0.3, group=SS, parameterName=SSOpacityInner}
  n63 = VectorParameter() {parameterName=SSColor, defaultValue=<rgba 0.7812 0.0229 0.1048 1>}
  n64 = Fresnel(ExponentIn=ScalarParameter() {defaultValue=1, group=SS, parameterName=SSFresnel}) {baseReflectFraction=0.005, exponent=8}
  n65 = Multiply(A=Multiply(A=@n35, B=@n63.RGB), B=@n64)
  n66 = Multiply(A=@n63.RGB, B=@n64)
  n67 = GetMaterialAttributes(None=MakeMaterialAttributes!(
    BaseColor=LinearInterpolate(A=@n43, B=@n44, Alpha=@n45)
    Metallic=Multiply(A=@n46.G, B=@n47)
    Specular=Multiply(A=@n46.B, B=@n48)
    Roughness=Multiply(A=@n46.R, B=@n49)
    Anisotropy
    EmissiveColor=Add(A=@n52, B=@n61)
    Opacity=@n62
    OpacityMask
    Normal=TextureSample() {samplerType=SAMPLERTYPE_Normal, texture=&/T_Gumdrop_N}.RGB
    Tangent
    WorldPositionOffset
    SubsurfaceColor=StaticSwitchParameter(True=@n65, False=@n66) {parameterName=IsCandyCorn}
    ClearCoat
    ClearCoatRoughness
    AmbientOcclusion
    Refraction
    CustomizedUVs_0
    CustomizedUVs_1
    CustomizedUVs_2
    CustomizedUVs_3
    CustomizedUVs_4
    CustomizedUVs_5
    CustomizedUVs_6
    CustomizedUVs_7
    PixelDepthOffset
    ShadingModel
    Displacement
   )) {attributeGetTypes [#1, #2, #3]}
  n68 = LinearInterpolate(B=ScalarParameter() {defaultValue=1, group=Sparkles, parameterName=Saprkle_GlowStrength}, Alpha=@n25)
  LinearInterpolate(
   A=Multiply(
    A=MaterialFunctionCall(Base=@n67.Base Color, Blend=@n27) {materialFunction=&Blend_Overlay}.Result
    B=@n68
   )
   Alpha=@n26
  )
  LinearInterpolate(
   A=@n62
   B=ScalarParameter() {group=SS, parameterName=SSOpacityOuter}
   Alpha=Power(Base=@n64, Exp=ScalarParameter() {defaultValue=1, group=SS, parameterName=SSOpacityPow})
  )
  Multiply(B=ScalarParameter() {defaultValue=0.7, parameterName=NormalsTile})
  n69 = ReflectionVectorWS()
  Multiply(
   A=Power(
    Base=TextureSampleParameterCube!(UVs=@n69, Apply View MipBias) {parameterName=ReflectionTexture, texture=/Game/Packages/Fortress_Sky/TexturesHDR/T_hdrsky_test03_low}.RGB
    Exp=ScalarParameter() {defaultValue=3.5, group=ReflectionMap, parameterName=ReflectionPower}
   ) {constExponent=3}
   B=ScalarParameter() {group=ReflectionMap, parameterName=ReflectionX}
  )
  n70 = ScalarParameter() {defaultValue=2, group=Sparkles, parameterName=Sparkles_Tile_Rate B}
  Reroute(None=MaterialFunctionCall(TextureObject=@n16, TextureSize=@n14, WorldPosition=Multiply(A=@n70, B=@n12), World Space Normal=@n15.PS, ProjectionTransitionContrast=@n14) {materialFunction=&WorldAlignedTexture}.XYZ Texture)
  SetMaterialAttributes!(
   MaterialAttributes=@n67.MaterialAttributes
   Base Color=Add(A=@n67.Base Color, B=@n68)
   Roughness=@n67.MaterialAttributes
   Emissive Color=@n67.Emissive Color
  ) {attributeSetTypes [#1, #2, #3]}
  StaticSwitchParameter(True=@n35, False=VectorParameter() {parameterName=Color, defaultValue=<rgba 0.7188 0.0137 0 1>}.RGB) {parameterName=IsCandyCorn}
  TextureCoordinate()
 ]
}
